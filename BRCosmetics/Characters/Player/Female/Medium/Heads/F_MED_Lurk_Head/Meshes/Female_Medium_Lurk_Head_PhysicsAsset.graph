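# Physics asset for the Lurk head mesh, a medium-build female player character cosmetic.
# It approximates the head and neck with two capsule bodies, one on the faceAttach bone and a larger one on neck_02.
# A single constraint joins the two bodies with equal, symmetric angular limits on both swing axes and twist.
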PhysicsAsset /BRCosmetics/Characters/Player/Female/Medium/Heads/F_MED_Lurk_Head/Meshes/Female_Medium_Lurk_Head_PhysicsAsset
bodies {
 faceAttach [{shapeType=Capsule, length=8.0135, radius=6.1476, center=<0.0042 4.5773 -0.4221>, rotation=<rot 7.9035 -90 -0.0515>}]
 neck_02 [{shapeType=Capsule, length=10.4879, radius=9.1072, center=<8.0593 0.9019 -0.2067>, rotation=<rot 7.9394 -90 -77.1126>}]
}
constraints [{bone1Name=faceAttach, bone2Name=neck_02, swing1LimitDegrees=45, swing2LimitDegrees=45, twistLimitDegrees=45}]
